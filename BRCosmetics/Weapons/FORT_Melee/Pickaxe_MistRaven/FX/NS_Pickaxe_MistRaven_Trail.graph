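# Niagara swing trail for the MistRaven pickaxe, built from anim-trail ribbons that follow socket pairs on the User skeletal mesh.
# A wide Dark ribbon lerps from pink-red to near-black by User.Variant, which also switches between Trail and Trail_Variant, each spawning four thin trails.
# Those trails use four ribbon renderers, loop-age-faded glow/core/center colors, and per-trail phase-offset material animation.
# Trail_Variant tints core and center blue, dimmed at night.
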NiagaraSystem /BRCosmetics/Weapons/FORT_Melee/Pickaxe_MistRaven/FX/NS_Pickaxe_MistRaven_Trail
dependencies {
 usedDataInterfaces [&NiagaraDataInterfaceArrayInt32, &NiagaraDataInterfaceCurve, &NiagaraDataInterfaceArrayFloat]
 usedDynamicInputs []
 usedModules [&SystemState]
}
emitters {
 Dark {
  emitterUpdateScript [
   &EmitterState(Life Cycle Mode=::NewEnumerator1, Loop Behavior=::NewEnumerator1, Loop Duration=0.25, UseLoopDelay=true)
   &FN_SpawnAnimTrail()
  ]
  particleSpawnScript [
   &FN_InitializeAnimTrailBoneIndecies(First Bone or Socket Index=2, Second Bone or Socket Index=3)
   &FN_InitializeAnimTrail(
    AnimTrailWidth=4
    Color=&LerpLinearColors(StartColor=<rgba 1 0.0162 0.1179 0.3>, EndColor=<rgba 0.0208 0.0168 0.0169 1.25>, LerpFactor=$User.Variant)
    Lifetime=0.6
   )
  ]
  particleUpdateScript [
   &FN_UpdateAnimTrail(SkeletalMesh=$User.SkeletalMesh)
   &ParticleState()
   &ScaleColor(
    Scale Alpha=&Multiply_Float(
     A=&Multiply_Float(
      A=&FloatFromCurve(
       FloatCurve=DataInterface() {propertyValues {Curve {keys [
           {arriveTangent=-0.0116, interpMode=RCIM_Cubic, leaveTangent=-3.1782, tangentMode=RCTM_Break, time=0.3, value=1}
           {arriveTangent=-0.029, interpMode=RCIM_Cubic, tangentMode=RCTM_Break, time=1, value=0}
          ]}}}
       Scale Curve=0.35
      )
      B=&FloatFromCurve(FloatCurve=DataInterface() {propertyValues {Curve {keys [<at 0.05 0>,<at 0.1 1>]}}}, CurveIndex=$Emitter.Age)
     )
     B=&FN_FrontEndFloatModifier(Gameplay=&Lerp_Float(A=1, B=1.2, Alpha=$User.Variant), Front End=0.5)
    )
    Scale RGB=&Lerp_Vector(
     A=<1 1 1>
     B=&FN_Select_Vector_from_Time_Of_Day(Morning=<1 1 1>, Day=<1 1 1>, Evening=<1 1 1>, Night=<0.25 0.25 0.25>)
     Alpha=$User.Variant
    )
    ScaleA=true
    ScaleRGB=true
   )
  ]
  rendererClasses [&NiagaraRibbonRendererProperties]
  renderers [{rendererClass=&NiagaraRibbonRendererProperties}]
 }
 Trail {
  emitterUpdateScript [
   &EmitterState(Life Cycle Mode=::NewEnumerator1, Loop Behavior=::NewEnumerator1, Loop Delay=0.05, Loop Duration=0.25, UseLoopDelay=true)
   &FN_SpawnAnimTrail(Number of Anim Trails:NiagaraInt32=&SelectIntFromArray(
     Direct Array Index=&MakeIntFromFloat(Float=$User.Variant)
     Int Selection Array=DataInterface() {propertyValues {IntData [4,0]}}
    ))
  ]
  particleSpawnScript [
   &FN_InitializeAnimTrailBoneIndecies()
   &FN_InitializeAnimTrailBoneIndecies(AnimTrail Selection=1, First Bone or Socket Index=2, Second Bone or Socket Index=3) {module=FN_InitializeAnimTrailBoneIndecies001}
   &FN_InitializeAnimTrailBoneIndecies(AnimTrail Selection=2, First Bone or Socket Index=4, Second Bone or Socket Index=5) {module=FN_InitializeAnimTrailBoneIndecies003}
   &FN_InitializeAnimTrailBoneIndecies(AnimTrail Selection=3, First Bone or Socket Index=6, Second Bone or Socket Index=7) {module=FN_InitializeAnimTrailBoneIndecies002}
   &FN_InitializeAnimTrail(AnimTrailWidth=0.7, Color=<rgba 1 0.2 0.35 0.5>)
  ]
  particleUpdateScript [
   &ParticleState()
   &FN_UpdateAnimTrail(SkeletalMesh=$User.SkeletalMesh)
   &ScaleRibbonWidth(Ribbon Width Scale=&FloatFromCurve(FloatCurve=DataInterface() {propertyValues {Curve {keys [{arriveTangent=-0.0042, interpMode=RCIM_Cubic, leaveTangent=-0.0042, tangentMode=RCTM_User, time=0, value=1}, <at 0.5 0>]}}}))
   &DynamicMaterialParameters(
    Index 0 Param 2=&Add_Float(
     A=&Multiply_Float(A=$Engine.RealTime, B=2)
     B=&SelectFloatFromArray(Direct Array Index=$Particles.SpawnGroup, Float Selection Array=DataInterface() {propertyValues {FloatData [0,0.25,0.5,0.75]}})
    )
    Index 0 Param 4=&FloatFromCurve(FloatCurve=DataInterface() {propertyValues {Curve {keys [{arriveTangent=-0.9325, time=0.4, value=1}, {arriveTangent=-1.7559, tangentMode=RCTM_User, time=1, value=0.3}]}}})
    Module.Index0Param0Write=true
    Module.Index0Param1Write=true
    Module.Index0Param2Write=true
    Module.Index0Param3Write=true
   )
   /SetVariables(
    Particles.RibbonWidth_Base:NiagaraFloat=&Multiply_Float(A=$Particles.RibbonWidth, B=2)
    Particles.RibbonWidth_Center:NiagaraFloat=&Multiply_Float(A=$Particles.RibbonWidth, B=1.5)
    Particles.RibbonWidth_Core:NiagaraFloat=&Multiply_Float(A=$Particles.RibbonWidth, B=1.5)
    Particles.RibbonWidth_Glow:NiagaraFloat=&Multiply_Float(A=$Particles.RibbonWidth, B=2)
   ) {isSetParameters=true, module=SetVariables_#1}
   /SetVariables(Particles.ColorGlow:LinearColor=&MakeLinearColorFromVectorAndFloat(
     "Vector (RGB)"=&MakeVectorFromLinearColorRGB(Color=<rgba 1 1 1 3>)
     "Float (Alpha)"=&Multiply_Float(
      A=0.3
      B=&FloatFromCurve(
       FloatCurve=DataInterface() {propertyValues {Curve {keys [<at 0 1>, {arriveTangent=-5.7875, leaveTangent=-5.7875, time=0.1, value=1}, <at 0.5 0>]}}}
       CurveIndex=$Emitter.NormalizedLoopAge
      )
     )
    )) {isSetParameters=true, module=SetVariables_#2}
   /SetVariables(Particles.ColorCore:LinearColor=&MakeLinearColorFromVectorAndFloat(
     "Vector (RGB)"=&MakeVectorFromLinearColorRGB(Color=<rgba 1 1 1 3>)
     "Float (Alpha)"=&Multiply_Float(
      A=1
      B=&FloatFromCurve(
       FloatCurve=DataInterface() {propertyValues {Curve {keys [<at 0 1>, {arriveTangent=-4.1873, leaveTangent=-4.1873, time=0.1, value=1}, <at 0.8 0>]}}}
       CurveIndex=$Emitter.NormalizedLoopAge
      )
     )
    )) {isSetParameters=true, module=SetVariables_#3}
   /SetVariables(Particles.ColorCenter:LinearColor=&MakeLinearColorFromVectorAndFloat(
     "Vector (RGB)"=&MakeVectorFromLinearColorRGB(Color=<rgba 1 1 1 3>)
     "Float (Alpha)"=&FloatFromCurve(
      FloatCurve=DataInterface() {propertyValues {Curve {keys [<at 0 1>, {arriveTangent=-5.7875, leaveTangent=-5.7875, time=0.1, value=1}, <at 0.5 0>]}}}
      CurveIndex=$Emitter.NormalizedLoopAge
     )
    )) {isSetParameters=true, module=SetVariables_#4}
   /SetVariables(Particles.VisibilityTag=&MakeIntFromFloat(Float=$User.Variant)) {isSetParameters=true, module=SetVariables_#5}
  ]
  rendererClasses [&NiagaraRibbonRendererProperties]
  renderers [
   {rendererClass=&NiagaraRibbonRendererProperties}
   {rendererClass=&NiagaraRibbonRendererProperties, rendererIndex=1}
   {rendererClass=&NiagaraRibbonRendererProperties, rendererIndex=2}
   {rendererClass=&NiagaraRibbonRendererProperties, rendererIndex=3}
  ]
 }
 Trail_Variant {
  emitterUpdateScript [
   &EmitterState(Life Cycle Mode=::NewEnumerator1, Loop Behavior=::NewEnumerator1, Loop Delay=0.05, Loop Duration=0.25, UseLoopDelay=true)
   &FN_SpawnAnimTrail(Number of Anim Trails:NiagaraInt32=&SelectIntFromArray(
     Direct Array Index=&MakeIntFromFloat(Float=$User.Variant)
     Int Selection Array=DataInterface() {propertyValues {IntData [0,4]}}
    ))
  ]
  particleSpawnScript [
   &FN_InitializeAnimTrailBoneIndecies()
   &FN_InitializeAnimTrailBoneIndecies(AnimTrail Selection=1, First Bone or Socket Index=2, Second Bone or Socket Index=3) {module=FN_InitializeAnimTrailBoneIndecies001}
   &FN_InitializeAnimTrailBoneIndecies(AnimTrail Selection=2, First Bone or Socket Index=4, Second Bone or Socket Index=5) {module=FN_InitializeAnimTrailBoneIndecies003}
   &FN_InitializeAnimTrailBoneIndecies(AnimTrail Selection=3, First Bone or Socket Index=6, Second Bone or Socket Index=7) {module=FN_InitializeAnimTrailBoneIndecies002}
   &FN_InitializeAnimTrail(AnimTrailWidth=0.7, Color=<rgba 1 0.2 0.35 0.5>)
  ]
  particleUpdateScript [
   &ParticleState()
   &FN_UpdateAnimTrail(SkeletalMesh=$User.SkeletalMesh)
   &ScaleRibbonWidth(Ribbon Width Scale=&FloatFromCurve(FloatCurve=DataInterface() {propertyValues {Curve {keys [{arriveTangent=-0.0042, interpMode=RCIM_Cubic, leaveTangent=-0.0042, tangentMode=RCTM_User, time=0, value=1}, <at 0.5 0>]}}}))
   &DynamicMaterialParameters(
    Index 0 Param 2=&Add_Float(
     A=&Multiply_Float(A=$Engine.RealTime, B=2)
     B=&SelectFloatFromArray(Direct Array Index=$Particles.SpawnGroup, Float Selection Array=DataInterface() {propertyValues {FloatData [0,0.25,0.5,0.75]}})
    )
    Index 0 Param 4=&FloatFromCurve(FloatCurve=DataInterface() {propertyValues {Curve {keys [{arriveTangent=-0.9325, time=0.4, value=1}, {arriveTangent=-1.7559, tangentMode=RCTM_User, time=1, value=0.3}]}}})
    Module.Index0Param0Write=true
    Module.Index0Param1Write=true
    Module.Index0Param2Write=true
    Module.Index0Param3Write=true
   )
   /SetVariables(
    Particles.RibbonWidth_Base:NiagaraFloat=&Multiply_Float(A=$Particles.RibbonWidth, B=2)
    Particles.RibbonWidth_Center:NiagaraFloat=&Multiply_Float(A=$Particles.RibbonWidth, B=1.5)
    Particles.RibbonWidth_Core:NiagaraFloat=&Multiply_Float(A=$Particles.RibbonWidth, B=1.5)
    Particles.RibbonWidth_Glow:NiagaraFloat=&Multiply_Float(A=$Particles.RibbonWidth, B=2)
   ) {isSetParameters=true, module=SetVariables_#1}
   /SetVariables(Particles.ColorGlow:LinearColor=&MakeLinearColorFromVectorAndFloat(
     "Vector (RGB)"=&MakeVectorFromLinearColorRGB(Color=<rgba 1 1 1 3>)
     "Float (Alpha)"=&Multiply_Float(
      A=0.3
      B=&FloatFromCurve(
       FloatCurve=DataInterface() {propertyValues {Curve {keys [<at 0 1>, {arriveTangent=-5.7875, leaveTangent=-5.7875, time=0.1, value=1}, <at 0.5 0>]}}}
       CurveIndex=$Emitter.NormalizedLoopAge
      )
     )
    )) {isSetParameters=true, module=SetVariables_#2}
   /SetVariables(Particles.ColorCore:LinearColor=&MakeLinearColorFromVectorAndFloat(
     "Vector (RGB)"=&FN_Select_Vector_from_Time_Of_Day(Morning=<0.8 1 2>, Day=<0.8 1 2>, Evening=<0.8 1 2>, Night=<0.28 0.35 0.7>)
     "Float (Alpha)"=&Multiply_Float(
      A=1
      B=&FloatFromCurve(
       FloatCurve=DataInterface() {propertyValues {Curve {keys [<at 0 1>, {arriveTangent=-4.1873, leaveTangent=-4.1873, time=0.1, value=1}, <at 0.8 0>]}}}
       CurveIndex=$Emitter.NormalizedLoopAge
      )
     )
    )) {isSetParameters=true, module=SetVariables_#3}
   /SetVariables(Particles.ColorCenter:LinearColor=&MakeLinearColorFromVectorAndFloat(
     "Vector (RGB)"=&FN_Select_Vector_from_Time_Of_Day(Morning=<0.8 1 2>, Day=<0.8 1 2>, Evening=<0.8 1 2>, Night=<0.28 0.35 0.7>)
     "Float (Alpha)"=&FloatFromCurve(
      FloatCurve=DataInterface() {propertyValues {Curve {keys [<at 0 1>, {arriveTangent=-5.7875, leaveTangent=-5.7875, time=0.1, value=1}, <at 0.5 0>]}}}
      CurveIndex=$Emitter.NormalizedLoopAge
     )
    )) {isSetParameters=true, module=SetVariables_#4}
   /SetVariables(Particles.VisibilityTag=&MakeIntFromFloat(Float=$User.Variant)) {isSetParameters=true, module=SetVariables_#5}
  ]
  rendererClasses [&NiagaraRibbonRendererProperties]
  renderers [
   {rendererClass=&NiagaraRibbonRendererProperties}
   {rendererClass=&NiagaraRibbonRendererProperties, rendererIndex=1}
   {rendererClass=&NiagaraRibbonRendererProperties, rendererIndex=2}
   {rendererClass=&NiagaraRibbonRendererProperties, rendererIndex=3}
  ]
 }
}
summary {
 systemName=NS_Pickaxe_MistRaven_Trail
 emitters [
  {emitterName=Trail, rendererClasses [&NiagaraRibbonRendererProperties]}
  {emitterName=Dark, rendererClasses [&NiagaraRibbonRendererProperties]}
  {emitterName=Trail_Variant, rendererClasses [&NiagaraRibbonRendererProperties]}
 ]
 userVariables [
  User.bIsFrontEnd {type=NiagaraBool}
  User.bIsFrontEndPreview {type=NiagaraBool}
  User.RibbonIDOffset {type=NiagaraInt32}
  User.SkeletalMesh {
   type=NiagaraDataInterfaceSkeletalMesh
   defaultValue {
    struct=NiagaraExt_VariableValue_DataInterface
    value {dataInterface=&self:NiagaraDataInterfaceSkeletalMesh_0}
   }
  }
  User.Variant {type=NiagaraFloat, defaultValue=0, description=""}
 ]
}
