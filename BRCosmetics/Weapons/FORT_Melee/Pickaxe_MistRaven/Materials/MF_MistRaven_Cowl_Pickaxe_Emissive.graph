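# Material function for the MistRaven Cowl pickaxe that overrides BaseColor and EmissiveColor on incoming material attributes.
# Base emissive combines a tint, optional texture cropping and squaring, brightness with a mobile multiplier, and switchable Fresnel.
# Hand masks and Fresnel terms blend inside colors into the emissive, while BaseColor is desaturated, tinted and mask-blended toward hand colors.
# It also outputs an Emissive Mask and an unconnected HandSmokeMask.
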MaterialFunction /BRCosmetics/Weapons/FORT_Melee/Pickaxe_MistRaven/Materials/MF_MistRaven_Cowl_Pickaxe_Emissive
graph {nodes [
  n1 = GetMaterialAttributes(None=FunctionInput() {inputName=Material, inputType=FunctionInput_MaterialAttributes}) {attributeGetTypes [#1]}
  n2 = NamedRerouteUsage()
  n3 = TextureSampleParameter2D() {
   group=06_Emissive
   parameterName=Hand_TA_mask
   samplerSource=SSM_Wrap_WorldGroupSettings
   samplerType=SAMPLERTYPE_Masks
   texture=/BRCosmetics/Weapons/FORT_Melee/Pickaxe_MistRaven/Textures/T_MistRaven_Pickaxe_FX2
  }
  n4 = Fresnel(
   ExponentIn=ScalarParameter() {defaultValue=0.2751, group=06_Emissive, parameterName=Exponentln_Pickaxe}
   BaseReflectFractionIn=ScalarParameter() {defaultValue=-1.4221, group=06_Emissive, parameterName=BaseReflectFractionln_pickaxe}
  ) {baseReflectFraction=0}
  n5 = LinearInterpolate(
   A=LinearInterpolate(
    A=VectorParameter() {group=MistRaven, parameterName=Inside_color2, defaultValue=<rgba 0.2135 0.0607 0.1448 1>}.RGB
    B=VectorParameter() {group=MistRaven, parameterName=Inside_color, defaultValue=<rgba 0.35 0.0391 0.0996 1>}.RGB
    Alpha=@n3.G
   )
   B=NamedRerouteUsage()
   Alpha=Power(Base=Saturate(None=@n4), Exp=ScalarParameter() {defaultValue=0.5299, group=06_Emissive, parameterName=InsidePower})
  )
  n6 = TextureSampleParameter2D() {
   group=06_Emissive
   parameterName=Hand_TA_mask
   samplerSource=SSM_Wrap_WorldGroupSettings
   samplerType=SAMPLERTYPE_Masks
   texture=&T_MistRaven_Pickaxe_FX2
  }
  n7 = OneMinus(None=Fresnel(
    ExponentIn=ScalarParameter() {defaultValue=0.8046, group=MistRaven, parameterName=Hand_Exponentln}
    BaseReflectFractionIn=ScalarParameter() {defaultValue=-0.0232, group=MistRaven, parameterName=Hand_BaseReflectFrationln}
   ) {exponent=1})
  n8 = ScalarParameter() {defaultValue=3.6305, parameterName=Hand_Fresnel_power}
  n9 = Multiply(A=@n6.G, B=Saturate(None=Power(Base=@n7, Exp=@n8)))
  FunctionOutput(None=SetMaterialAttributes!(MaterialAttributes=@n1.MaterialAttributes, EmissiveColor=LinearInterpolate(A=@n2, B=@n5, Alpha=@n9), BaseColor=NamedRerouteUsage()) {attributeSetTypes [#2, #1]})
  n10 = Constant() {r=1}
  n11 = NamedRerouteUsage()
  n12 = ScalarParameter() {defaultValue=0.85, group=06_Emissive, parameterName=Fresnel_Processing_power}
  FunctionOutput(None=Multiply(A=@n10, B=Power(Base=@n11, Exp=@n12) {constExponent=10})) {outputName=Emissive Mask}
  FunctionOutput() {outputName=HandSmokeMask}
  n13 = ScalarParameter() {defaultValue=1, group=06_Emissive, parameterName=EmissiveMobileMultiplier}
  n14 = Constant() {r=1}
  NamedRerouteDeclaration(None=Multiply(A=ScalarParameter() {defaultValue=1, group=06_Emissive, parameterName=EmissiveBrightness}, B=ShadingPathSwitch(Default=@n13, Deferred=@n14))) {name=Emissive Brightness Modulation, nodeColor=<rgba 1 0.3281 0 1>}
  n15 = Fresnel(ExponentIn=Constant() {r=1}) {baseReflectFraction=0}
  n16 = StaticSwitchParameter(True=@n15, False=OneMinus(None=@n15)) {defaultValue=true, group=06_Emissive, parameterName=InvertEmissiveFresnel}
  n17 = ScalarParameter() {defaultValue=0.8545, group=06_Emissive, parameterName=EmissiveFresnelPower}
  NamedRerouteDeclaration(None=Saturate(None=Power(Base=@n16, Exp=@n17))) {name=Fresnel Processing, nodeColor=<rgba 0 0.7187 1 1>}
  n18 = TextureObjectParameter() {group=06_Emissive, parameterName=EmissiveCropped, sortPriority=1, texture=&T_BlacksRGB}
  n19 = VectorParameter() {
   group=06_Emissive
   parameterName=CroppedEmissiveUVs
   sortPriority=2
   channelNames {a=Bottom Right Corner V, b=Bottom Right Corner U, g=Upper Left Corner V, r=Upper Left Corner U}
   defaultValue=<rgba 0 0 1 1>
  }
  NamedRerouteDeclaration(None=MaterialFunctionCall!(
    UVs
    TextureIn=@n18
    UpperLeftCorner=ComponentMask(None=@n19) {g=true, r=true}
    LowerRightCorner=AppendVector(A=@n19.Bottom Right Corner U, B=@n19.Bottom Right Corner U)
    ExportFloat4
   ) {materialFunction=/Engine/Functions/Engine_MaterialFunctions01/Texturing/TextureCropping}.Cropped) {name=CroppedEmissive, nodeColor=<rgba 1 0.4688 0 1>}
  NamedRerouteDeclaration(None=@n10) {name=Emissive Mask, nodeColor=<rgba 1 0 0.75 1>}
  n20 = VectorParameter() {group=MistRaven, parameterName=EmissiveTint_neck, sortPriority=6, defaultValue=<rgba 0.4522 0.0226 0.0413 1>}
  n21 = StaticSwitchParameter(True=NamedRerouteUsage(), False=NamedRerouteUsage()) {desc=CROP_EMIS, group=06_Emissive, parameterName=UseCroppedEmissive}
  n22 = Multiply(A=@n21, B=@n21)
  n23 = Multiply(
   A=Multiply(A=@n20.RGB, B=StaticSwitchParameter(True=@n22, False=@n21) {group=06_Emissive, parameterName=UseLinearEmissiveMask})
   B=NamedRerouteUsage()
  )
  n24 = NamedRerouteUsage()
  NamedRerouteDeclaration(None=StaticSwitchParameter(True=Multiply(A=@n23, B=@n24), False=@n23) {defaultValue=true, group=06_Emissive, parameterName=UseEmissiveFresnel_Cowl_Body}) {name=BaseEmissive Color, nodeColor=<rgba 1 0.8906 0 1>}
  n25 = Multiply(
   A=Desaturation(None=@n1.BaseColor, Fraction=Constant() {r=1})
   B=VectorParameter() {group=06_Emissive, parameterName=BaseColor_Mul_Color, defaultValue=<rgba 0.7758 0.5395 0.4621 1>}.RGB
  ) {constB=8}
  n26 = VectorParameter() {group=MistRaven, parameterName=Hand_Color2, defaultValue=<rgba 0.1 0.007 0.0292 1>}
  n27 = NamedRerouteUsage()
  n28 = TextureSample() {samplerType=SAMPLERTYPE_Masks, texture=/BRCosmetics/Weapons/FORT_Melee/Pickaxe_MistRaven/Textures/T_MistRaven_Pickaxe_FX}
  NamedRerouteDeclaration(None=LinearInterpolate(
    A=LinearInterpolate(A=@n25, B=@n26.RGB, Alpha=@n27)
    B=VectorParameter() {group=MistRaven, parameterName=Hand_Color, defaultValue=<rgba 0.1652 0.0244 0.1164 1>}.RGB
    Alpha=Power(Base=@n28.G) {constExponent=1.5}
   )) {name=outputBaseColor, nodeColor=<rgba 1 0 0.0703 1>}
 ]}
