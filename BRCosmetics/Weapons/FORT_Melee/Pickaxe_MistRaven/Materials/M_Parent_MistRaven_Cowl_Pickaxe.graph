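# Parent material for the MistRaven Cowl pickaxe cosmetic, a subsurface-shaded surface built from material attributes for skeletal meshes with morph targets and mesh deformers.
# It wraps pickaxe-specific effects between the shared 2023 character parent PreFX and PostFX functions.
# The PreFX result feeds a Cowl emissive function, whose output drives a transition function that also takes pre-skinned local position before passing to PostFX.
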Material /BRCosmetics/Weapons/FORT_Melee/Pickaxe_MistRaven/Materials/M_Parent_MistRaven_Cowl_Pickaxe
graph {
 properties {
  bUseMaterialAttributes=true
  bUsedWithMeshDeformer=true
  bUsedWithMorphTargets=true
  bUsedWithSkeletalMesh=true
  pixelDepthOffsetMode=PDOM_Legacy
  shadingModel=MSM_Subsurface
  usedShadingModels=Subsurface
 }
 nodes [
  Constant() {r=1}
  n1 = MaterialFunctionCall() {materialFunction=&MF_CharacterParent_2023_PreFX}
  n2 = MaterialFunctionCall!(Material=@n1.Result) {materialFunction=&/MF_MistRaven_Cowl_Pickaxe_Emissive}
  MaterialFunctionCall(
   Material=MaterialFunctionCall!(EmissiveTextureSampler=@n2.Emissive Mask, Material=@n2.Emissive Mask, PreSkinnedLocation=@n1.Pre skinned Local Position) {materialFunction=&/MF_Misraven_Pickaxe_Transition}.Result
   M=@n1.M
   Pre Skinned Local Position=@n1.M
   Distance Blend=@n1.M
   Fresnel=@n1.M
  ) {materialFunction=&MF_CharacterParent_2023_PostFX}
 ]
}
