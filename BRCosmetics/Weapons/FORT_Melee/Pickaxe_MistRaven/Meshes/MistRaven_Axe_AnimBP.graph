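# Animation Blueprint for the Mist Raven pickaxe that adjusts the axe's size and placement to suit the wielder's body type and gender.
# When character and pickaxe customization loads, one custom part's permitted body type and gender set four booleans: Large or not, Male or Female.
# The AnimGraph chains four Modify Bone nodes on b_AR_Root after the input pose, each boolean enabling one additive translation and scale offset.
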AnimBlueprint /BRCosmetics/Weapons/FORT_Melee/Pickaxe_MistRaven/Meshes/MistRaven_Axe_AnimBP
target_skeleton=&/MistRaven_Axe_Skeleton
graphs [
 AnimGraph {nodes [
   AnimGraphNode_SubInput_0 = Input Pose()
   n1 = "Transform (Modify) Bone - Bone: b_AR_Root"!(
    ComponentPose=Local To Component(LocalPose=@AnimGraphNode_SubInput_0)
    bAlphaBoolEnabled:bool=Get not Large Female() -> (not Large Female:bool)
    Translation="2,0,0"
    Rotation="0,0,0"
    Scale="1.01,1.01,1.01"
   ) -> (Pose) {
    alphaInputType=Bool
    scaleMode=BMM_Additive
    translationMode=BMM_Additive
    alphaBoolBlend {blendOption=Linear}
    boneToModify {boneName=b_AR_Root}
    scale=<1.01 1.01 1.01>
    translation {x=2}
   }
   n2 = "Transform (Modify) Bone - Bone: b_AR_Root"(ComponentPose=@n1, bAlphaBoolEnabled=Get not Large() -> (not Large:bool), Translation="0,-1,0", Rotation="0,0,0", Scale="1.02,1.02,1.02") {
    alphaInputType=Bool
    scaleMode=BMM_Additive
    translationMode=BMM_Additive
    alphaBoolBlend {blendOption=Linear}
    boneToModify {boneName=b_AR_Root}
    scale=<1.02 1.02 1.02>
    translation {y=-1}
   }
   n3 = "Transform (Modify) Bone - Bone: b_AR_Root"(
    ComponentPose=@n2
    bAlphaBoolEnabled=Get Large Female() -> (Large Female:bool)
    Translation="0,-1,0"
    Rotation="0,0,0"
    Scale="1.02,1.02,1.02"
   ) {
    alphaInputType=Bool
    scaleMode=BMM_Additive
    translationMode=BMM_Additive
    alphaBoolBlend {blendOption=Linear}
    boneToModify {boneName=b_AR_Root}
    scale=<1.02 1.02 1.02>
    translation {y=-1}
   }
   n4 = "Transform (Modify) Bone - Bone: b_AR_Root"(
    ComponentPose=@n3
    bAlphaBoolEnabled=Get Large Character() -> (Large Character:bool)
    Translation="0,-1,3.5"
    Rotation="0,0,0"
    Scale="1.1,1.1,1.1"
   ) {
    alphaInputType=Bool
    scaleMode=BMM_Additive
    translationMode=BMM_Additive
    alphaBoolBlend {blendOption=Linear}
    boneToModify {boneName=b_AR_Root}
    scale=<1.1 1.1 1.1>
    translation {y=-1, z=3.5}
   }
   Output Pose(Result=Component To Local(ComponentPose=@n4)) {layerGroup=None}
  ]}
 EventGraph {nodes [
   n1 = Event OnCharacterAndPickaxeCustomizationLoaded!() -> (
    OutputDelegate
    then
    CustomBaseMesh:Skeletal Mesh Component Object Reference
    CustomParts:Array of Custom Character Part Object References
    CustomMeshParts:Array of Skeletal Mesh Component Budgeted Object References
   ) {is=title K2Node_Event}
   n2 = Get Array Item with EFort Custom Part Type(Array:Array of Custom Character Part Object References=@n1.CustomParts) -> (Output:Custom Character Part Object Reference)
   n3 = "Equal (Enum)"(
    A:EFortCustomBodyType Enum=Get BodyTypesPermitted!(self:Custom Character Part Object Reference=@n2) -> (BodyTypesPermitted:EFortCustomBodyType Enum=NONE)
    B:EFortCustomBodyType Enum=Large
   )
   n4 = NOT Boolean(A=@n3)
   n5 = "Equal (Enum)"(
    A:EFortCustomGender Enum=Get GenderPermitted!(self:Custom Character Part Object Reference=@n2) -> (GenderPermitted:EFortCustomGender Enum=Invalid)
    B:EFortCustomGender Enum=Male
   )
   > Set not Large(not Large:bool=AND Boolean(A=@n4, B=@n5)) -> (Output_Get:bool)
   n6 = "Equal (Enum)"(A=Get GenderPermitted(self=@n2) -> (GenderPermitted=Invalid), B=Female)
   > Set not Large Female(not Large Female:bool=AND Boolean(A=@n4, B=@n6)) -> (Output_Get:bool)
   > Set Large Character(Large Character:bool=AND Boolean(A=@n3, B=@n5)) -> (Output_Get:bool)
   > Set Large Female(Large Female:bool=AND Boolean(A=@n3, B=@n6)) -> (Output_Get:bool)
   TryGetPawnOwner()
   Event BlueprintUpdateAnimation()
  ]}
]
variables [
 Large Character {type=boolean}
 Large Female {type=boolean}
 not Large {type=boolean}
 not Large Female {type=boolean}
]
